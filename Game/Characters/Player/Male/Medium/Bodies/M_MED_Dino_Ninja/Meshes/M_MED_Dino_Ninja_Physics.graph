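# Physics asset for the medium male Dino Ninja player body, set up for a dangling back flap on the costume.
# Capsules cover the pelvis, spine, thighs and hands, linked by free angular constraints, while boxes form the dyn_backFlap chain.
# The flap's rings body hangs from spine_03 with twist and one swing axis locked.
# The five segments below chain in sequence with limited swing and very tight twist.
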PhysicsAsset /Game/Characters/Player/Male/Medium/Bodies/M_MED_Dino_Ninja/Meshes/M_MED_Dino_Ninja_Physics
bodies {
 dyn_backFlap [{shapeType=Box, extentX=3.8629, extentY=13.4662, extentZ=2.5828, center=<0.047 -0.2338 -2.2186>, rotation=<rot -79.9972 -91.079 1.1502>}]
 dyn_backFlap1 [{shapeType=Box, extentX=3.8629, extentY=13.4662, extentZ=2.5828, center=<0.047 -0.2338 -2.2186>, rotation=<rot -84.9961 -92.1561 2.2227>}]
 dyn_backFlap2 [{shapeType=Box, extentX=3.8629, extentY=13.4662, extentZ=2.5828, center=<0.0476 -0.537 -2.1373>, rotation=<rot -74.9977 -90.646 0.7337>}]
 dyn_backFlap3 [{shapeType=Box, extentX=3.3629, extentY=12.9662, extentZ=2.0828, center=<0.0447 -0.3461 -2.8554>, rotation=<rot -79.9967 -91.1072 1.173>}]
 dyn_backFlap4 [{shapeType=Box, extentX=3.8629, extentY=13.4662, extentZ=2.5828, center=<0.0466 -0.3441 -2.3636>, rotation=<rot -79.9967 -91.2083 1.2639>}]
 dyn_backFlap_rings [{shapeType=Box, extentX=2.3629, extentY=13.4662, extentZ=2.5828, center=<0.0501 -0.2339 -0.8749>, rotation=<rot -64.9979 -90.1938 0.3365>}]
 hand_l [{shapeType=Capsule, length=6.8098, radius=6.2458, center=<7.2325 -0.6705 0.576>, rotation=<rot -6.2692 -75.586 -94.6763>}]
 hand_r [{shapeType=Capsule, length=6.0413, radius=6.4967, center=<-6.208 -0.1874 -0.2474>, rotation=<rot -65.0653 101.6853 87.3983>}]
 pelvis [{shapeType=Capsule, length=5.6245, radius=12.9789, center=<0.4132 0.2181 0.0447>, rotation=<rot -3.1054 -90.0108 177.6608>}]
 spine_01 [{shapeType=Capsule, length=3.005, radius=11.755, center=<-0.0868 0.2471 0>, rotation=<rot 0 -90 0>}]
 spine_02 [{shapeType=Capsule, length=6.7282, radius=10.3606, center=<-0.0099 0.047 -0.3454>, rotation=<rot 0.3945 97.201 178.916>}]
 spine_03 [{shapeType=Capsule, length=2.5, radius=6.6797, center=<-0.0285 -0.3829 -0.0827>, rotation=<rot 0 -82.8289 -178.977>}]
 spine_04 [{shapeType=Capsule, length=5.25, radius=9.3986, center=<0.2168 -0.1608 -0.2719>, rotation=<rot -0.7903 -174.57 179.9776>}]
 spine_05 [{shapeType=Capsule, length=20.125, radius=10.315, center=<-0.8312 -4.2653 1.3786>, rotation=<rot 0 -156.3347 -88.156>}]
 thigh_l [{shapeType=Capsule, length=23.4108, radius=9.3223, center=<-18.5622 -0.0164 -1.1735>, rotation=<rot -2.0042 -86.9702 90.9943>}]
 thigh_r [{shapeType=Capsule, length=22.2918, radius=9.1631, center=<17.1743 0.8473 0.3658>, rotation=<rot 10.2465 -88.3934 -94.0356>}]
}
constraints [
 {
  bone1Name=spine_01
  bone2Name=pelvis
  swing1LimitDegrees=45
  swing1Motion=Free
  swing2LimitDegrees=45
  swing2Motion=Free
  twistLimitDegrees=45
  twistMotion=Free
 }
 {
  bone1Name=spine_03
  bone2Name=spine_02
  swing1LimitDegrees=45
  swing1Motion=Free
  swing2LimitDegrees=45
  swing2Motion=Free
  twistLimitDegrees=45
  twistMotion=Free
 }
 {
  bone1Name=hand_l
  bone2Name=spine_05
  swing1LimitDegrees=45
  swing1Motion=Free
  swing2LimitDegrees=45
  swing2Motion=Free
  twistLimitDegrees=45
  twistMotion=Free
 }
 {
  bone1Name=hand_r
  bone2Name=spine_05
  swing1LimitDegrees=45
  swing1Motion=Free
  swing2LimitDegrees=45
  swing2Motion=Free
  twistLimitDegrees=45
  twistMotion=Free
 }
 {
  bone1Name=thigh_l
  bone2Name=pelvis
  swing1LimitDegrees=45
  swing1Motion=Free
  swing2LimitDegrees=45
  swing2Motion=Free
  twistLimitDegrees=45
  twistMotion=Free
 }
 {
  bone1Name=thigh_r
  bone2Name=pelvis
  swing1LimitDegrees=45
  swing1Motion=Free
  swing2LimitDegrees=45
  swing2Motion=Free
  twistLimitDegrees=45
  twistMotion=Free
 }
 {
  bone1Name=dyn_backFlap_rings
  bone2Name=spine_03
  swing1LimitDegrees=1
  swing1Motion=Locked
  swing2LimitDegrees=45
  twistLimitDegrees=1
  twistMotion=Locked
 }
 {bone1Name=dyn_backFlap, bone2Name=dyn_backFlap_rings, swing1LimitDegrees=45, swing2LimitDegrees=45, twistLimitDegrees=2}
 {bone1Name=dyn_backFlap1, bone2Name=dyn_backFlap, swing1LimitDegrees=45, swing2LimitDegrees=45, twistLimitDegrees=2}
 {bone1Name=dyn_backFlap2, bone2Name=dyn_backFlap1, swing1LimitDegrees=45, swing2LimitDegrees=45, twistLimitDegrees=2}
 {bone1Name=dyn_backFlap3, bone2Name=dyn_backFlap2, swing1LimitDegrees=45, swing2LimitDegrees=45, twistLimitDegrees=2}
 {bone1Name=dyn_backFlap4, bone2Name=dyn_backFlap3, swing1LimitDegrees=45, swing2LimitDegrees=45, twistLimitDegrees=2}
 {
  bone1Name=spine_02
  bone2Name=spine_01
  swing1LimitDegrees=45
  swing1Motion=Free
  swing2LimitDegrees=45
  swing2Motion=Free
  twistLimitDegrees=45
  twistMotion=Free
 }
 {
  bone1Name=spine_04
  bone2Name=spine_03
  swing1LimitDegrees=45
  swing1Motion=Free
  swing2LimitDegrees=45
  swing2Motion=Free
  twistLimitDegrees=45
  twistMotion=Free
 }
 {
  bone1Name=spine_05
  bone2Name=spine_04
  swing1LimitDegrees=45
  swing1Motion=Free
  swing2LimitDegrees=45
  swing2Motion=Free
  twistLimitDegrees=45
  twistMotion=Free
 }
]
